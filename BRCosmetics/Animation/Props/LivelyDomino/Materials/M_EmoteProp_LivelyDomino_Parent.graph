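# Parent material for the LivelyDomino emote prop, adding an animated color glow to the shared 2023 character material functions.
# Scrolled time drives a looping three-color blend, scaled by intensity and masked by the FX texture's red channel for emissive.
# Named reroutes feed base color and emissive overrides, alpha-blended between the PreFX and PostFX functions.
# The local-position X gradient connects to nothing; only time drives the loop.
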Material /BRCosmetics/Animation/Props/LivelyDomino/Materials/M_EmoteProp_LivelyDomino_Parent
graph {
 properties {
  bUseMaterialAttributes=true
  bUsedWithMeshDeformer=true
  bUsedWithSkeletalMesh=true
  pixelDepthOffsetMode=PDOM_Legacy
  previewMesh=/BRCosmetics/Animation/Props/Gadgets_LivelyDomino/Meshes/LivelyDomino_Gadget
  shadingModel=MSM_FromMaterialExpression
  usedShadingModels=Default Lit | Subsurface | Clear Coat | Subsurface Profile | Cloth
 }
 nodes [
  n1 = LocalPosition()
  n2 = Multiply(A=Time(), B=ScalarParameter() {defaultValue=-0.2, group=06_Emissive, parameterName=GradientScrollSpeed}) {constB=-1}
  Add(
   A=Divide(
    A=ComponentMask(None=@n1.XYZ) {r=true}
    B=ScalarParameter() {defaultValue=64, group=06_Emissive, parameterName=GradientSize, sliderMax=128, sliderMin=16}
   ) {constB=64}
   B=@n2
  )
  n3 = MaterialFunctionCall() {materialFunction=&MF_CharacterParent_2023_PreFX}
  n4 = NamedRerouteUsage()
  n5 = NamedRerouteUsage()
  MaterialFunctionCall(
   Material=BlendMaterialAttributes(
    A=@n3.Result
    B=SetMaterialAttributes!(MaterialAttributes=@n3.Result, Emissive Color=@n4, Base Color=@n5) {attributeSetTypes [#1, #2]}
    Alpha=NamedRerouteUsage()
   )
   M=@n3.M
   Pre Skinned Local Position=@n3.M
   Distance Blend=@n3.M
   Fresnel=@n3.M
  ) {materialFunction=&MF_CharacterParent_2023_PostFX}
  n6 = MaterialFunctionCall(Float3=TextureSample() {samplerType=SAMPLERTYPE_Masks, texture=/BRCosmetics/Animation/Props/LivelyDomino/Textures/T_EmoteProp_LivelyDomino_FX}.RGB) {materialFunction=&BreakOutFloat3Components}
  n7 = Max(A=@n6.R, B=@n6.R)
  n8 = NamedRerouteDeclaration(None=MaterialFunctionCall!(A, B, C, Alpha=Frac(None=@n2)) {materialFunction=/Game/Global/MaterialFunctions/TechArt/MF_3ColorBlendLoop}.Result) {name=BaseColor_Out, nodeColor=<rgba 1 0.7892 0 1>}
  n9 = ScalarParameter() {defaultValue=20, group=06_Emissive, parameterName=EmissiveIntensity, sliderMax=100}
  NamedRerouteDeclaration(None=Multiply(A=NamedRerouteDeclaration(None=@n7) {name=Glow_Mask, nodeColor=<rgba 0 0.8008 1 1>}, B=Multiply(A=@n8, B=@n9))) {name=Emissive_Out, nodeColor=<rgba 0 1 0.3906 1>}
 ]
}
